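# Physics asset for the Temple pickaxe melee weapon, rigging a knot with two dangling strings for secondary motion.
# One capsule body covers the two-handed handle, one the knot, and five small ones each string chain.
# The knot rotates freely relative to the handle, and both strings hang from it.
# String roots have wide swing limits, later segments tighter ones, and all string joints allow little twist.
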PhysicsAsset /BRCosmetics/Weapons/FORT_Melee/Pickaxe_Temple/Meshes/Temple_Axe_Physics
bodies {
 Melee_TwoHanded_Handle [{shapeType=Capsule, length=33.7377, radius=16.8064, center=<-0.4475 -4.6706 13.1682>, rotation=<rot -3.6302 -90 -0.0036>}]
 dyn_knot [{shapeType=Capsule, length=9.755, radius=2.755, center=<-0.18 -0.0002 3.0234>, rotation=<rot 0 -89.9992 42.4081>}]
 dyn_string_l_00 [{shapeType=Capsule, length=1.8883, radius=0.6399, center=<1.9769 0 0>, rotation=<rot 90 0 0>}]
 dyn_string_l_01 [{shapeType=Capsule, length=1.9031, radius=0.625, center=<1.9769 0 0>, rotation=<rot 90 0 0>}]
 dyn_string_l_02 [{shapeType=Capsule, length=1.8883, radius=0.6399, center=<1.9769 0 0>, rotation=<rot 90 0 0>}]
 dyn_string_l_03 [{shapeType=Capsule, length=1.8883, radius=0.6399, center=<1.9769 0 0>, rotation=<rot 90 0 0>}]
 dyn_string_l_04 [{shapeType=Capsule, length=1.8883, radius=0.6399, center=<1.9769 0 0>, rotation=<rot 90 0 0>}]
 dyn_string_r_00 [{shapeType=Capsule, length=1.8883, radius=0.6399, center=<1.9769 0 0>, rotation=<rot 90 0 0>}]
 dyn_string_r_01 [{shapeType=Capsule, length=1.8883, radius=0.6399, center=<1.9769 0 0>, rotation=<rot 90 0 0>}]
 dyn_string_r_02 [{shapeType=Capsule, length=1.8883, radius=0.6399, center=<1.9769 0 0>, rotation=<rot 90 0 0>}]
 dyn_string_r_03 [{shapeType=Capsule, length=1.8883, radius=0.6399, center=<1.9769 0 0>, rotation=<rot 90 0 0>}]
 dyn_string_r_04 [{shapeType=Capsule, length=1.8883, radius=0.6399, center=<1.9769 0 0>, rotation=<rot 90 0 0>}]
}
constraints [
 {bone1Name=dyn_string_r_00, bone2Name=dyn_knot, swing1LimitDegrees=160, swing2LimitDegrees=120, twistLimitDegrees=1}
 {bone1Name=dyn_string_r_01, bone2Name=dyn_string_r_00, swing1LimitDegrees=45, swing2LimitDegrees=45, twistLimitDegrees=5}
 {bone1Name=dyn_string_r_02, bone2Name=dyn_string_r_01, swing1LimitDegrees=45, swing2LimitDegrees=45, twistLimitDegrees=5}
 {bone1Name=dyn_string_r_03, bone2Name=dyn_string_r_02, swing1LimitDegrees=45, swing2LimitDegrees=45, twistLimitDegrees=5}
 {bone1Name=dyn_string_l_00, bone2Name=dyn_knot, swing1LimitDegrees=160, swing2LimitDegrees=120, twistLimitDegrees=5}
 {bone1Name=dyn_string_l_01, bone2Name=dyn_string_l_00, swing1LimitDegrees=45, swing2LimitDegrees=45, twistLimitDegrees=5}
 {bone1Name=dyn_string_l_02, bone2Name=dyn_string_l_01, swing1LimitDegrees=45, swing2LimitDegrees=45, twistLimitDegrees=5}
 {bone1Name=dyn_string_l_03, bone2Name=dyn_string_l_02, swing1LimitDegrees=45, swing2LimitDegrees=45, twistLimitDegrees=5}
 {
  bone1Name=dyn_knot
  bone2Name=Melee_TwoHanded_Handle
  swing1LimitDegrees=45
  swing1Motion=Free
  swing2LimitDegrees=45
  swing2Motion=Free
  twistLimitDegrees=45
  twistMotion=Free
 }
 {bone1Name=dyn_string_l_04, bone2Name=dyn_string_l_03, swing1LimitDegrees=45, swing2LimitDegrees=45, twistLimitDegrees=5}
 {bone1Name=dyn_string_r_04, bone2Name=dyn_string_r_03, swing1LimitDegrees=45, swing2LimitDegrees=45, twistLimitDegrees=5}
]
